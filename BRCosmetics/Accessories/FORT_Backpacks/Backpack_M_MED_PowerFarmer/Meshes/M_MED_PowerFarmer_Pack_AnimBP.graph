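# Post-processes the PowerFarmer back bling's pose, animating a creature-like pack with wandering eyes and dangling parts.
# Randomized, delayed EventGraph logic drives interpolated alphas on additive inner-eye, mid-eye and pupil bone modifiers; outer-eye modifier chains are unconnected to the output.
# AnimDynamics then jiggles tail, jaw, legs and ears.
# Finally, the owner's spine delta from reference pose offsets and rolls the root, weakened in the front end.
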AnimBlueprint /BRCosmetics/Accessories/FORT_Backpacks/Backpack_M_MED_PowerFarmer/Meshes/M_MED_PowerFarmer_Pack_AnimBP
target_skeleton=&/M_MED_PowerFarmer_Pack_Skeleton
graphs [
 AnimGraph {nodes [
   n1 = "Transform (Modify) Bone - Bone: dyn_l_eye"(Alpha=Get eye_fwd() -> (eye_fwd:double), Translation="0,0,0", Rotation="0.25,0,0", Scale="1,1,1") {
    rotationMode=BMM_Additive
    rotationSpace=BCS_BoneSpace
    alphaBoolBlend {blendOption=Linear}
    alphaScaleBiasClamp {bInterpResult=true, interpSpeedDecreasing=15, interpSpeedIncreasing=15}
    boneToModify {boneName=dyn_l_eye}
    rotation {pitch=0.25}
   }
   "Transform (Modify) Bone - Bone: dyn_l_eye"(ComponentPose=@n1, Alpha=Get eye_bck() -> (eye_bck:double), Translation="0,0,0", Rotation="-0.75,0,0", Scale="1,1,1") {
    rotationMode=BMM_Additive
    rotationSpace=BCS_BoneSpace
    alphaBoolBlend {blendOption=Linear}
    alphaScaleBiasClamp {bInterpResult=true, interpSpeedDecreasing=15, interpSpeedIncreasing=15}
    boneToModify {boneName=dyn_l_eye}
    rotation {pitch=-0.75}
   }
   n2 = "Transform (Modify) Bone - Bone: dyn_r_eye"(Alpha=Get eye_fwd(), Translation="0,0,0", Rotation="0,-0.75,0", Scale="1,1,1") {
    rotationMode=BMM_Additive
    rotationSpace=BCS_BoneSpace
    alphaBoolBlend {blendOption=Linear}
    alphaScaleBiasClamp {bInterpResult=true, interpSpeedDecreasing=15, interpSpeedIncreasing=15}
    boneToModify {boneName=dyn_r_eye}
    rotation {yaw=-0.75}
   }
   "Transform (Modify) Bone - Bone: dyn_r_eye"(ComponentPose=@n2, Alpha=Get eye_bck(), Translation="0,0,0", Rotation="0,0.75,0", Scale="1,1,1") {
    rotationMode=BMM_Additive
    rotationSpace=BCS_BoneSpace
    alphaBoolBlend {blendOption=Linear}
    alphaScaleBiasClamp {bInterpResult=true, interpSpeedDecreasing=15, interpSpeedIncreasing=15}
    boneToModify {boneName=dyn_r_eye}
    rotation {yaw=0.75}
   }
   n3 = "Transform (Modify) Bone - Bone: dyn_l_eye"(Alpha=Get eye_fwd(), Translation="0,0,0", Rotation="0,-0.25,0", Scale="1,1,1") {
    rotationMode=BMM_Additive
    rotationSpace=BCS_BoneSpace
    alphaBoolBlend {blendOption=Linear}
    alphaScaleBiasClamp {bInterpResult=true, interpSpeedDecreasing=15, interpSpeedIncreasing=15}
    boneToModify {boneName=dyn_l_eye}
    rotation {yaw=-0.25}
   }
   "Transform (Modify) Bone - Bone: dyn_l_eye"(ComponentPose=@n3, Alpha=Get eye_bck(), Translation="0,0,0", Rotation="0,0.25,0", Scale="1,1,1") {
    rotationMode=BMM_Additive
    rotationSpace=BCS_BoneSpace
    alphaBoolBlend {blendOption=Linear}
    alphaScaleBiasClamp {bInterpResult=true, interpSpeedDecreasing=15, interpSpeedIncreasing=15}
    boneToModify {boneName=dyn_l_eye}
    rotation {yaw=0.25}
   }
   n4 = "Transform (Modify) Bone - Bone: dyn_r_eye"(Alpha=Get eye_fwd(), Translation="0,0,0", Rotation="0.25,0,0", Scale="1,1,1") {
    rotationMode=BMM_Additive
    rotationSpace=BCS_BoneSpace
    alphaBoolBlend {blendOption=Linear}
    alphaScaleBiasClamp {bInterpResult=true, interpSpeedDecreasing=15, interpSpeedIncreasing=15}
    boneToModify {boneName=dyn_r_eye}
    rotation {pitch=0.25}
   }
   "Transform (Modify) Bone - Bone: dyn_r_eye"(ComponentPose=@n4, Alpha=Get eye_bck(), Translation="0,0,0", Rotation="-0.75,0,0", Scale="1,1,1") {
    rotationMode=BMM_Additive
    rotationSpace=BCS_BoneSpace
    alphaBoolBlend {blendOption=Linear}
    alphaScaleBiasClamp {bInterpResult=true, interpSpeedDecreasing=15, interpSpeedIncreasing=15}
    boneToModify {boneName=dyn_r_eye}
    rotation {pitch=-0.75}
   }
   "Transform (Modify) Bone - Bone: dyn_l_innerEye"(Alpha=Get InnerRotateBackward() -> (InnerRotateBackward:double), Translation="0,0,0", Rotation="0,50,0", Scale="1,1,1") {
    rotationMode=BMM_Additive
    rotationSpace=BCS_BoneSpace
    alphaBoolBlend {blendOption=Linear}
    alphaScaleBiasClamp {bInterpResult=true, interpSpeedDecreasing=20, interpSpeedIncreasing=15}
    boneToModify {boneName=dyn_l_innerEye}
    rotation {yaw=50}
   }
   n5 = "Transform (Modify) Bone - Bone: dyn_l_innerEye"(
    ComponentPose=Local To Component(LocalPose=Input Pose())
    Alpha=Get InnerRotateForward() -> (InnerRotateForward:double)
    Translation="0,0,0"
    Rotation="-15,0,0"
    Scale="1,1,1"
   ) {
    rotationMode=BMM_Additive
    rotationSpace=BCS_BoneSpace
    alphaBoolBlend {blendOption=Linear}
    alphaScaleBiasClamp {bInterpResult=true, interpSpeedDecreasing=20, interpSpeedIncreasing=15}
    boneToModify {boneName=dyn_l_innerEye}
    rotation {pitch=-15}
   }
   n6 = "Transform (Modify) Bone - Bone: dyn_l_innerEye"(ComponentPose=@n5, Alpha=Get InnerRotateBackward(), Translation="0,0,0", Rotation="15,0,0", Scale="1,1,1") {
    rotationMode=BMM_Additive
    rotationSpace=BCS_BoneSpace
    alphaBoolBlend {blendOption=Linear}
    alphaScaleBiasClamp {bInterpResult=true, interpSpeedDecreasing=15, interpSpeedIncreasing=20}
    boneToModify {boneName=dyn_l_innerEye}
    rotation {pitch=15}
   }
   n7 = "Transform (Modify) Bone - Bone: dyn_l_MidEye"(ComponentPose=@n6, Alpha=Get MidRotate Backward() -> (MidRotate Backward:double), Translation="0,0,0", Rotation="0,0,-22", Scale="1,1,1") {
    rotationMode=BMM_Additive
    rotationSpace=BCS_BoneSpace
    alphaBoolBlend {blendOption=Linear}
    alphaScaleBiasClamp {bInterpResult=true, interpSpeedDecreasing=15, interpSpeedIncreasing=15}
    boneToModify {boneName=dyn_l_MidEye}
    rotation {roll=-22}
   }
   n8 = "Transform (Modify) Bone - Bone: dyn_l_pupilShink"(ComponentPose=@n7, Alpha=Get InnerScaleUp() -> (InnerScaleUp:double), Translation="0,0,0", Rotation="0,0,0", Scale="1.1,1.1,1.1") {
    rotationMode=BMM_Additive
    rotationSpace=BCS_BoneSpace
    scaleMode=BMM_Additive
    scaleSpace=BCS_BoneSpace
    alphaBoolBlend {blendOption=Linear}
    alphaScaleBiasClamp {bInterpResult=true, interpSpeedDecreasing=5, interpSpeedIncreasing=5}
    boneToModify {boneName=dyn_l_pupilShink}
    scale=<1.1 1.1 1.1>
   }
   n9 = "Transform (Modify) Bone - Bone: dyn_l_innerEye"(ComponentPose=@n8, Alpha=Get InnerRotateForward(), Translation="0,0,0", Rotation="0,-15,0", Scale="1,1,1") {
    rotationMode=BMM_Additive
    rotationSpace=BCS_BoneSpace
    alphaBoolBlend {blendOption=Linear}
    alphaScaleBiasClamp {bInterpResult=true, interpSpeedDecreasing=15, interpSpeedIncreasing=20}
    boneToModify {boneName=dyn_l_innerEye}
    rotation {yaw=-15}
   }
   n10 = "Transform (Modify) Bone - Bone: dyn_l_MidEye"(ComponentPose=@n9, Alpha=Get MidRotateForward() -> (MidRotateForward:double), Translation="0,0,0", Rotation="0,0,22", Scale="1,1,1") {
    rotationMode=BMM_Additive
    rotationSpace=BCS_BoneSpace
    alphaBoolBlend {blendOption=Linear}
    alphaScaleBiasClamp {bInterpResult=true, interpSpeedDecreasing=15, interpSpeedIncreasing=15}
    boneToModify {boneName=dyn_l_MidEye}
    rotation {roll=22}
   }
   n11 = "Transform (Modify) Bone - Bone: dyn_l_pupilShink"(ComponentPose=@n10, Alpha=Get InnerScaleDown() -> (InnerScaleDown:double), Translation="0,0,0", Rotation="0,0,0", Scale="0.65,0.65,0.65") {
    rotationMode=BMM_Additive
    rotationSpace=BCS_BoneSpace
    scaleMode=BMM_Additive
    scaleSpace=BCS_BoneSpace
    alphaBoolBlend {blendOption=Linear}
    alphaScaleBiasClamp {bInterpResult=true, interpSpeedDecreasing=5, interpSpeedIncreasing=5}
    boneToModify {boneName=dyn_l_pupilShink}
    scale=<0.65 0.65 0.65>
   }
   n12 = "Transform (Modify) Bone - Bone: dyn_r_innerEye"(ComponentPose=@n11, Alpha=Get InnerRotateForward(), Translation="0,0,0", Rotation="-23,0,0", Scale="1,1,1") {
    rotationMode=BMM_Additive
    rotationSpace=BCS_BoneSpace
    alphaBoolBlend {blendOption=Linear}
    alphaScaleBiasClamp {bInterpResult=true, interpSpeedDecreasing=20, interpSpeedIncreasing=17}
    boneToModify {boneName=dyn_r_innerEye}
    rotation {pitch=-23}
   }
   n13 = "Transform (Modify) Bone - Bone: dyn_r_innerEye"(ComponentPose=@n12, Alpha=Get InnerRotateBackward(), Translation="0,0,0", Rotation="23,0,0", Scale="1,1,1") {
    rotationMode=BMM_Additive
    rotationSpace=BCS_BoneSpace
    alphaBoolBlend {blendOption=Linear}
    alphaScaleBiasClamp {bInterpResult=true, interpSpeedDecreasing=20, interpSpeedIncreasing=17}
    boneToModify {boneName=dyn_r_innerEye}
    rotation {pitch=23}
   }
   n14 = "Transform (Modify) Bone - Bone: dyn_r_MidEye"(ComponentPose=@n13, Alpha=Get MidRotate Backward(), Translation="0,0,0", Rotation="0,0,-22", Scale="1,1,1") {
    rotationMode=BMM_Additive
    rotationSpace=BCS_BoneSpace
    alphaBoolBlend {blendOption=Linear}
    alphaScaleBiasClamp {bInterpResult=true}
    boneToModify {boneName=dyn_r_MidEye}
    rotation {roll=-22}
   }
   n15 = "Transform (Modify) Bone - Bone: dyn_r_pupilShink"(ComponentPose=@n14, Alpha=Get InnerScaleUp(), Translation="0,0,0", Rotation="0,0,0", Scale="1.2,1.2,1.2") {
    rotationMode=BMM_Additive
    rotationSpace=BCS_BoneSpace
    scaleMode=BMM_Additive
    scaleSpace=BCS_BoneSpace
    alphaBoolBlend {blendOption=Linear}
    alphaScaleBiasClamp {bInterpResult=true, interpSpeedDecreasing=5, interpSpeedIncreasing=5}
    boneToModify {boneName=dyn_r_pupilShink}
    scale=<1.2 1.2 1.2>
   }
   n16 = "Transform (Modify) Bone - Bone: dyn_r_innerEye"(ComponentPose=@n15, Alpha=Get InnerRotateUp() -> (InnerRotateUp:double), Translation="0,0,0", Rotation="0,-23,0", Scale="1,1,1") {
    rotationMode=BMM_Additive
    rotationSpace=BCS_BoneSpace
    alphaBoolBlend {blendOption=Linear}
    alphaScaleBiasClamp {bInterpResult=true, interpSpeedDecreasing=20, interpSpeedIncreasing=17}
    boneToModify {boneName=dyn_r_innerEye}
    rotation {yaw=-23}
   }
   n17 = "Transform (Modify) Bone - Bone: dyn_r_innerEye"(ComponentPose=@n16, Alpha=Get InnerRotateDown() -> (InnerRotateDown:double), Translation="0,0,0", Rotation="0,23,0", Scale="1,1,1") {
    rotationMode=BMM_Additive
    rotationSpace=BCS_BoneSpace
    alphaBoolBlend {blendOption=Linear}
    alphaScaleBiasClamp {bInterpResult=true, interpSpeedDecreasing=20, interpSpeedIncreasing=17}
    boneToModify {boneName=dyn_r_innerEye}
    rotation {yaw=23}
   }
   n18 = "Transform (Modify) Bone - Bone: dyn_r_MidEye"(ComponentPose=@n17, Alpha=Get MidRotateForward(), Translation="0,0,0", Rotation="0,0,22", Scale="1,1,1") {
    rotationMode=BMM_Additive
    rotationSpace=BCS_BoneSpace
    alphaBoolBlend {blendOption=Linear}
    alphaScaleBiasClamp {bInterpResult=true}
    boneToModify {boneName=dyn_r_MidEye}
    rotation {roll=22}
   }
   n19 = "Transform (Modify) Bone - Bone: dyn_r_pupilShink"(ComponentPose=@n18, Alpha=Get InnerScaleDown(), Translation="0,0,0", Rotation="0,0,0", Scale="0.55,0.55,0.55") {
    rotationMode=BMM_Additive
    rotationSpace=BCS_BoneSpace
    scaleMode=BMM_Additive
    scaleSpace=BCS_BoneSpace
    alphaBoolBlend {blendOption=Linear}
    alphaScaleBiasClamp {bInterpResult=true, interpSpeedDecreasing=5, interpSpeedIncreasing=5}
    boneToModify {boneName=dyn_r_pupilShink}
    scale=<0.55 0.55 0.55>
   }
   n20 = Anim Dynamics - Bone: dyn_tail_01(ComponentPose=@n19, Alpha=1, ExternalForce="0,0,0") {
    angularDampingOverride=0.999
    angularSpringConstant=2000
    bAngularSpring=true
    bGravityOverrideInSimSpace=true
    bOverrideAngularDamping=true
    gravityScale=0.8
    simulationSpace=RootRelative
    alphaBoolBlend {blendOption=Linear}
    boundBone {boneName=dyn_tail_01}
    componentAppliedLinearAccClamp=<2250 2250 2250>
    componentLinearAccScale=<1 1 1>
    physicsBodyDefinitions [{
      sphereCollisionRadius=0
      boundBone {boneName=dyn_tail_01}
      boxExtents=<3 10 10>
      constraintSetup {angularLimitsMax=<0 10 0>, angularLimitsMin=<0 -60 0>, angularTarget=<0 0 0>}
      localJointOffset=<1.5 0 0>
     }]
   }
   n21 = Anim Dynamics - Bone: dyn_tail_02(ComponentPose=@n20, Alpha=0.6, ExternalForce="0,0,0") {
    angularDampingOverride=0.999
    angularSpringConstant=2000
    bAngularSpring=true
    bGravityOverrideInSimSpace=true
    bOverrideAngularDamping=true
    simulationSpace=RootRelative
    alphaBoolBlend {blendOption=Linear}
    boundBone {boneName=dyn_tail_02}
    componentAppliedLinearAccClamp=<5000 5000 5000>
    componentLinearAccScale=<1 1 1>
    physicsBodyDefinitions [{
      sphereCollisionRadius=0
      boundBone {boneName=dyn_tail_02}
      boxExtents=<9 16 16>
      constraintSetup {angularLimitsMax=<0 10 0>, angularLimitsMin=<0 -60 0>, angularTarget=<0 0 0>}
      localJointOffset=<1.5 0 0>
     }]
   }
   n22 = Anim Dynamics - Bone: dyn_jaw(ComponentPose=@n21, Alpha=0.3, ExternalForce="0,0,0") {
    angularDampingOverride=1
    angularSpringConstant=20
    bGravityOverrideInSimSpace=true
    bOverrideAngularDamping=true
    gravityScale=0.2
    simulationSpace=BoneRelative
    alphaBoolBlend {blendOption=Linear}
    boundBone {boneName=dyn_jaw}
    componentAppliedLinearAccClamp=<2500 2500 2500>
    componentLinearAccScale=<1 1 1>
    gravityOverride=<0.7 0.7 0.7>
    physicsBodyDefinitions [{
      sphereCollisionRadius=0
      boundBone {boneName=dyn_jaw}
      boxExtents=<10 10 5>
      constraintSetup {angularLimitsMax=<0 50 0>, angularLimitsMin=<0 -30 0>, angularTarget=<0 0 0>}
      localJointOffset=<5 0 -3>
     }]
    relativeSpaceBone {boneName=root}
   }
   n23 = Anim Dynamics - Bone: dyn_br_leg(ComponentPose=@n22, Alpha=0.6, ExternalForce="0,0,0") {
    angularDampingOverride=1
    angularSpringConstant=1500
    bAngularSpring=true
    bGravityOverrideInSimSpace=true
    bOverrideAngularDamping=true
    gravityScale=0.6
    simulationSpace=RootRelative
    alphaBoolBlend {blendOption=Linear}
    boundBone {boneName=dyn_br_leg}
    componentAppliedLinearAccClamp=<5000 5000 5000>
    componentLinearAccScale=<0.5 0.5 0.5>
    physicsBodyDefinitions [{
      sphereCollisionRadius=0
      boundBone {boneName=dyn_br_leg}
      boxExtents=<3 10 10>
      constraintSetup {angularLimitsMax=<0 60 0>, angularLimitsMin=<0 -60 0>, angularTarget=<0 0 0>}
      localJointOffset=<1.5 0 0>
     }]
   }
   n24 = Anim Dynamics - Bone: dyn_fr_leg(ComponentPose=@n23, Alpha=0.53, ExternalForce="0,0,0") {
    angularDampingOverride=1
    angularSpringConstant=1500
    bAngularSpring=true
    bGravityOverrideInSimSpace=true
    bOverrideAngularDamping=true
    bUsePlanarLimit=false
    gravityScale=0.3
    simulationSpace=RootRelative
    alphaBoolBlend {blendOption=Linear}
    boundBone {boneName=dyn_fr_leg}
    componentAppliedLinearAccClamp=<5000 5000 5000>
    componentLinearAccScale=<0.25 0.25 0.25>
    physicsBodyDefinitions [{
      sphereCollisionRadius=0
      boundBone {boneName=dyn_fr_leg}
      boxExtents=<5 12 12>
      constraintSetup {angularLimitsMax=<0 60 0>, angularLimitsMin=<0 -60 0>, angularTarget=<0 0 0>}
      localJointOffset=<3 0 0>
     }]
   }
   n25 = Anim Dynamics - Bone: dyn_bl_leg(ComponentPose=@n24, Alpha=0.5, ExternalForce="0,0,0") {
    angularDampingOverride=1
    angularSpringConstant=1500
    bAngularSpring=true
    bGravityOverrideInSimSpace=true
    bOverrideAngularDamping=true
    gravityScale=0.4
    simulationSpace=RootRelative
    alphaBoolBlend {blendOption=Linear}
    boundBone {boneName=dyn_bl_leg}
    componentAppliedLinearAccClamp=<5000 5000 5000>
    componentLinearAccScale=<0.3 0.3 0.3>
    physicsBodyDefinitions [{
      sphereCollisionRadius=0
      boundBone {boneName=dyn_bl_leg}
      boxExtents=<4 11 11>
      constraintSetup {angularLimitsMax=<0 60 0>, angularLimitsMin=<0 -60 0>, angularTarget=<0 0 0>}
      localJointOffset=<2 0 0>
     }]
   }
   n26 = Anim Dynamics - Bone: dyn_fl_leg(ComponentPose=@n25, Alpha=0.43, ExternalForce="0,0,0") {
    angularDampingOverride=1
    angularSpringConstant=1500
    bAngularSpring=true
    bGravityOverrideInSimSpace=true
    bOverrideAngularDamping=true
    gravityScale=0.45
    simulationSpace=RootRelative
    alphaBoolBlend {blendOption=Linear}
    boundBone {boneName=dyn_fl_leg}
    componentAppliedLinearAccClamp=<5000 5000 5000>
    componentLinearAccScale=<0.2 0.2 0.2>
    physicsBodyDefinitions [{
      sphereCollisionRadius=0
      boundBone {boneName=dyn_fl_leg}
      boxExtents=<3 10 10>
      constraintSetup {angularLimitsMax=<0 60 0>, angularLimitsMin=<0 -60 0>, angularTarget=<0 0 0>}
      localJointOffset=<3.3 0 0>
     }]
   }
   n27 = Anim Dynamics - Bone: dyn_l_ear(ComponentPose=@n26, Alpha=0.5, ExternalForce="0,0,0") {
    angularDampingOverride=0.9999
    bAngularSpring=true
    bGravityOverrideInSimSpace=true
    bOverrideAngularDamping=true
    simulationSpace=RootRelative
    alphaBoolBlend {blendOption=Linear}
    boundBone {boneName=dyn_l_ear}
    componentAppliedLinearAccClamp=<3000 3000 3000>
    componentLinearAccScale=<0.5 0.5 0.5>
    physicsBodyDefinitions [{
      sphereCollisionRadius=0
      boundBone {boneName=dyn_l_ear}
      boxExtents=<5 12 12>
      constraintSetup {angularLimitsMax=<0 60 0>, angularLimitsMin=<0 -60 0>, angularTarget=<0 0 0>}
      localJointOffset=<-1.5 0 0>
     }]
   }
   n28 = Anim Dynamics - Bone: dyn_r_ear(ComponentPose=@n27, Alpha=0.23, ExternalForce="0,0,0") {
    angularDampingOverride=0.9999
    bAngularSpring=true
    bGravityOverrideInSimSpace=true
    bOverrideAngularDamping=true
    simulationSpace=RootRelative
    alphaBoolBlend {blendOption=Linear}
    boundBone {boneName=dyn_r_ear}
    componentAppliedLinearAccClamp=<5000 5000 5000>
    componentLinearAccScale=<0.4 0.4 0.4>
    physicsBodyDefinitions [{
      sphereCollisionRadius=0
      boundBone {boneName=dyn_r_ear}
      boxExtents=<3 10 10>
      constraintSetup {angularLimitsMax=<0 60 0>, angularLimitsMin=<0 -60 0>, angularTarget=<0 0 0>}
      localJointOffset=<-1.5 0 0>
     }]
   }
   n29 = "Transform (Modify) Bone - Bone: root"(
    ComponentPose=@n28
    Alpha=Get PackAlpha() -> (PackAlpha:double)
    Translation=Get BackpackTranslate() -> (BackpackTranslate:Vector)
    Rotation=Get BackpackRotate() -> (BackpackRotate:Rotator)
   ) {
    rotationMode=BMM_Additive
    rotationSpace=BCS_BoneSpace
    translationMode=BMM_Additive
    translationSpace=BCS_BoneSpace
    alphaBoolBlend {blendOption=Linear}
    boneToModify {boneName=root}
   }
   Output Pose(Result=Component To Local(ComponentPose=@n29)) {layerGroup=None}
  ]}
 EventGraph {nodes [
   Event BlueprintUpdateAnimation()
   > n1 = Sequence!(execute) -> (then_0, then_1, then_2:Exec, then_3:Exec, then_4:Exec)
   Branch(execute=@n1.then_1, Condition=Get InnerRotateForward_0() -> (InnerRotateForward_0:bool))
   > Delay(Duration=0.2)
   > Set eye_fwd(eye_fwd:double=Get InnerRotateForward()) -> (Output_Get:double)
   Branch(execute=@n1.then_2, Condition=Get InnerRotateBackwardActual() -> (InnerRotateBackwardActual:bool))
   > Delay(Duration=0.2)
   > Set eye_bck(eye_bck:double=Get InnerRotateBackward()) -> (Output_Get:double)
   n2 = Branch(execute=@n1.then_3, Condition=Get InnerRotateForward_0())
   > Delay(Duration=2)
   > Set MidRotateForward(MidRotateForward:double=Get InnerRotateForward()) -> (Output_Get:double)
   > Set MidEyeRotation(MidEyeRotation:bool=true) -> (Output_Get:bool)
   > Delay(Duration=0.5)
   > Set MidRotate Backward(MidRotate Backward:double=0) -> (Output_Get:double)
   > Delay(Duration=1)
   > Set MidRotateForward(MidRotateForward=Get InnerRotateForward())
   > n3 = Branch(Condition=Get MidEyeRotation() -> (MidEyeRotation:bool))
   n4 = RandomFloatInRange(Min=0.1, Max=1)
   > Set InnerScaleUp(InnerScaleUp:double=@n4) -> (Output_Get:double)
   > Set InnerScaleDown(InnerScaleDown:double=@n4) -> (Output_Get:double)
   Set InnerScaleUp(execute=@n3.else, InnerScaleUp=@n4)
   > Set InnerScaleDown(InnerScaleDown=@n4)
   Delay(execute=@n2.else, Duration=2)
   > Set MidRotateForward(MidRotateForward=0)
   > Set MidEyeRotation(MidEyeRotation=false)
   > Delay(Duration=0.5)
   > Set MidRotate Backward(MidRotate Backward=Get InnerRotateBackward())
   > Delay(Duration=1)
   > Set MidRotateForward(MidRotateForward=0)
   n5 = GetOwnerSkeletalMeshComponent()
   n6 = BreakRotator(InRot=BreakTransform(InTransform=GetDeltaTransformFromRefPose(self=@n5, BoneName=spine_05, BaseName=spine_01)).Rotation)
   Set BackpackRotate(execute=@n1.then_4, BackpackRotate:Rotator=MakeRotator(Roll=MapRangeClamped(Value=@n6.Yaw, InRangeA=0, InRangeB=-30, OutRangeB=35))) -> (Output_Get:Rotator)
   > Set BackpackTranslate(BackpackTranslate:Vector=MakeVector(Y=MapRangeClamped(Value=@n6.Yaw, InRangeA=0, InRangeB=-30, OutRangeB=-10))) -> (Output_Get:Vector)
   > n7 = Branch(Condition=Get bIsInFrontEnd())
   > Set PackAlpha(PackAlpha:double=0.15) -> (Output_Get:double)
   Set PackAlpha(execute=@n7.else, PackAlpha=1)
   Delay(execute=@n1.then_0, Duration=RandomFloatInRange(Min=0.65, Max=2.25))
   > n8 = Branch(Condition=RandomBool!() -> (ReturnValue:bool))
   n9 = RandomFloatInRange(Min=0.1, Max=1)
   > Set InnerRotateForward(InnerRotateForward:double=@n9) -> (Output_Get:double)
   > Set InnerRotateForward_0(InnerRotateForward_0:bool="float > float"(A=Get InnerRotateForward(), B=0.1)) -> (Output_Get:bool)
   > Set InnerRotateBackward(InnerRotateBackward:double=0) -> (Output_Get:double)
   > Set InnerRotateBackwardActual(InnerRotateBackwardActual:bool=false) -> (Output_Get:bool)
   > Set InnerRotateUp(InnerRotateUp:double=@n9) -> (Output_Get:double)
   > Set InnerEyeMovement(InnerEyeMovement:double=@n9) -> (Output_Get:double)
   Set InnerRotateForward(execute=@n8.else, InnerRotateForward=0)
   > Set InnerRotateForward_0(InnerRotateForward_0=false)
   > Set InnerRotateBackward(InnerRotateBackward=@n9)
   > Set InnerRotateBackwardActual(InnerRotateBackwardActual="float > float"(A=Get InnerRotateBackward(), B=0.1))
   > Set InnerRotateDown(InnerRotateDown:double=0) -> (Output_Get:double)
   > Set InnerEyeMovement(InnerEyeMovement=@n9)
   TryGetPawnOwner()
   Set MidRotate Backward(MidRotate Backward=0)
  ]}
]
variables [
 BackpackRotate {type=Rotator}
 BackpackTranslate {type=Vector}
 eye_bck {type=number, default=0.1619}
 eye_fwd {type=number}
 EyeActualMovement {type=number}
 InnerEyeMovement {type=number}
 InnerRotateBackward {type=number}
 InnerRotateBackwardActual {type=boolean}
 InnerRotateDown {type=number, default=0.3972}
 InnerRotateForward {type=number, default=0.3972}
 InnerRotateForward_0 {type=boolean}
 InnerRotateUp {type=number, default=0.3972}
 InnerScaleDown {type=number}
 InnerScaleUp {type=number, default=0.8557}
 MidEyeMovemnet {type=number}
 MidEyeRotation {type=boolean}
 MidRotate Backward {type=number, default=0.5975}
 MidRotateForward {type=number, default=0.5708}
 PackAlpha {type=number}
]
